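# Masked unlit material for a simple 2D Creative device icon, built from a tintable texture with a selectable opacity channel.
# A TwoSidedSign lerp mirrors UVs on back faces, and static switches choose camera-facing alignment and an optional sine bob for world position offset.
# The graph also holds a two-axis versus pivot-rotation facing switch and a hologram switch adding radial displacement from flickering world-aligned noise and a height pulse.
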Material /Game/Creative/Devices/Generic/Materials/M_CP_Device_2DIcon_Simple
graph {
 properties {blendMode=BLEND_Masked, pixelDepthOffsetMode=PDOM_Legacy, shadingModel=MSM_Unlit, usedShadingModels=Unlit}
 nodes [
  n1 = TextureCoordinate()
  n2 = Constant2Vector() {g=1, r=-1}
  n3 = TwoSidedSign()
  n4 = TextureSampleParameter2D(UVs=LinearInterpolate(A=Multiply(A=@n1, B=@n2) {constB=-1}, B=@n1, Alpha=Clamp(None=@n3))) {
   group=Opacity
   parameterName=Opacity Texture
   sortPriority=5
   texture=/Game/UI/Foundation/Textures/Icons/Replay/T-Icon-Replay-CurrentPlayer-L
  }
  n5 = ChannelMaskParameter(None=@n4.RGBA) {group=Opacity, maskChannel=Alpha, parameterName=Opacity Channel, sortPriority=6, defaultValue=<rgba 0 0 0 1>}
  Multiply(A=ScalarParameter() {defaultValue=1, desc=Check Blend Mode, group=Opacity, parameterName=Master Opacity, sortPriority=10}, B=@n5.R)
  n6 = MaterialFunctionCall!(
   WorldPosition
   PivotLocation
   WS Normals
   Custom Object Basis 1=Constant3Vector() {constant=<rgba 1 0 0 0>}
   Custom Object Basis 2=Constant3Vector() {constant=<rgba 0 0 -1 0>}
   Custom Object Basis 3=Constant3Vector() {constant=<rgba 0 1 0 0>}
  ) {materialFunction=/Engine/Functions/Engine_MaterialFunctions02/WorldPositionOffset/AlignMeshToTheCamera}
  n7 = StaticSwitchParameter(True=@n6.World Position Offset, False=Constant()) {defaultValue=true, group=WPO, parameterName=Is Camera Facing}
  n8 = VectorParameter() {group=WPO, parameterName=BobHeightMax, defaultValue=<rgba 0 0 10 0>}
  n9 = VectorParameter() {group=WPO, parameterName=BobHeightMin, defaultValue=<rgba 0 0 -2 0>}
  n10 = Sine(None=Multiply(A=Time(), B=ScalarParameter() {defaultValue=0.4, group=WPO, parameterName=WPOTimeScale}))
  n11 = StaticSwitchParameter(True=Add(A=@n7, B=LinearInterpolate(A=@n8.RGB, B=@n9.RGB, Alpha=@n10)), False=@n7) {group=WPO, parameterName=Bob?}
  Reroute(None=@n11)
  n12 = VectorParameter() {group=Color, parameterName=Color, sortPriority=11, defaultValue=<rgba 0.9047 1 1 1>}
  StaticSwitchParameter(True=Multiply(A=@n12.RGB, B=@n4.RGB), False=@n12.RGB) {defaultValue=true, parameterName=UseTextureForEmissive}
  n13 = CameraPositionWS()
  n14 = Constant3Vector() {constant=<rgba 1 1 0 0>}
  MaterialExpressionNormalize_0 = Normalize(VectorInput=Subtract(A=ObjectPositionWS(), B=Multiply(A=@n13, B=@n14)))
  n15 = Custom!(In=@MaterialExpressionNormalize_0) {
   code="float2 output; \r\nfloat2 Input1; \r\noutput = atan2 (In.y,In.x); \r\nreturn (output);"
   desc="float2 output; \r\nfloat2 Input1; \r\noutput = atan2 (In.y,In.x); \r\nreturn (output);"
   outputType=CMOT_Float2
   inputs [{inputName=In, input {expression=&self:MaterialExpressionNormalize_0}}]
  }
  n16 = Divide(A=Divide(A=@n15) {constB=3.142})
  n17 = MaterialFunctionCall() {materialFunction=&ObjectPivotPoint}
  n18 = VectorParameter() {group=WPO, parameterName=Pivot Offset, defaultValue=<rgba 0 0 0 1>}
  StaticSwitchParameter(
   True=@n6.World Position Offset
   False=RotateAboutAxis(
    NormalizedRotationAxis=ObjectOrientation()
    RotationAngle=Add(A=@n16) {constB=0.25}
    PivotPoint=Add(A=@n17.Object Pivot Location, B=@n18.RGB)
    Position=WorldPosition().XYZ
   )
  ) {group=WPO, parameterName=Use 2-Axis, sortPriority=1}
  n19 = Divide(A=ObjectRadius(), B=ScalarParameter() {defaultValue=0.1, group=Noise, parameterName=NoiseSize})
  n20 = Multiply(A=Time(), B=VectorParameter() {group=Noise, parameterName=NoisePanSpeed, defaultValue=<rgba -0.08 0.07 -0.2 0>}.RGB)
  n21 = MaterialFunctionCall(
   TextureObject=TextureObject() {samplerType=SAMPLERTYPE_LinearColor, texture=&T_NoisePack_LinearColor}
   TextureSize=Multiply(A=VectorParameter() {group=Noise, parameterName=NoiseTileMult, defaultValue=<rgba 1 1 0.1 1>}.RGB, B=@n19)
   WorldPosition=Add(A=WorldPosition().XYZ, B=Multiply(A=@n19, B=@n20))
  ) {materialFunction=&WorldAlignedTexture}
  n22 = ActorPositionWS()
  n23 = MaterialFunctionCall(Float2=Round(None=Divide(A=@n22) {constB=2000})) {materialFunction=&BreakOutFloat2Components}
  n24 = Add(A=Time(), B=Add(A=@n23.R, B=@n23.R))
  n25 = Multiply(A=ScalarParameter() {defaultValue=0.5, group=Hologram, parameterName=FlickerSpeed}, B=@n24)
  n26 = AppendVector(A=Sine(None=@n25) {period=1.6}, B=Sine(None=@n25) {period=1.42})
  n27 = Sine(None=@n25) {period=1.51}
  n28 = Sine(None=@n25) {period=2.1}
  n29 = ScalarParameter() {defaultValue=0.5, group=Noise, parameterName=NoiseFrequency}
  n30 = Multiply(B=@n29) {constA=0.25}
  n31 = Add(A=Sine(None=@n25) {period=0.72}, B=Multiply(B=@n29) {constA=0.75}) {constB=0.75}
  n32 = Add(A=Multiply(B=@n29) {constA=0.9, constB=0.9}, B=Sine(None=@n25) {period=1.25}) {constB=0.9}
  n33 = Multiply(A=Multiply(A=Add(A=@n28, B=@n30), B=@n31), B=Multiply(A=@n31, B=Ceil(None=@n32)))
  n34 = ConstantBiasScale(None=Sine(None=@n24) {period=0.075})
  n35 = Multiply(A=Ceil(None=AppendVector(A=@n26, B=@n27)), B=Multiply(A=Saturate(None=@n33), B=LinearInterpolate(Alpha=@n34) {constA=0.7}))
  n36 = ScalarParameter() {defaultValue=2, group=Noise, parameterName=NoiseMult}
  n37 = Multiply(
   A=Multiply(A=Power(Base=@n21.XY Texture) {constExponent=4}, B=ScalarParameter() {defaultValue=2, group=Noise, parameterName=NoiseMaskMult}) {constB=20}
   B=Add(A=ScalarParameter() {group=Noise, parameterName=NoiseAdd}, B=Multiply(A=@n35, B=@n36))
  )
  n38 = ComponentMask(None=MaterialFunctionCall() {materialFunction=&ObjectPivotPoint}.Object Pivot Location) {b=true}
  n39 = ScalarParameter() {defaultValue=300, group=Hologram, parameterName=PulseHeight}
  n40 = ScalarParameter() {defaultValue=1, group=Hologram, parameterName=TurnOn}
  n41 = ScalarParameter() {defaultValue=33, group=Hologram, parameterName=PulseRadius}
  n42 = SphereMask(
   A=ComponentMask(None=WorldPosition().XYZ) {b=true}
   B=LinearInterpolate(A=@n38, B=Add(A=@n38, B=@n39) {constB=1600}, Alpha=@n40) {constA=-0.25, constB=1.2}
   Radius=LinearInterpolate(A=@n41, B=Multiply(A=@n41) {constB=0.1}, Alpha=@n40) {constA=500, constB=100}
  ) {hardnessPercent=0}
  n43 = OneMinus(None=@n40)
  n44 = Add(A=Desaturation(None=Saturate(None=@n37)), B=Multiply(A=Multiply(A=@n42, B=@n43)) {constB=2})
  n45 = Subtract(A=WorldPosition().XYZ, B=ObjectPositionWS())
  n46 = ObjectRadius()
  n47 = ScalarParameter() {defaultValue=0.025, group=Noise, parameterName=WPOStrengthFromObjectRadius}
  n48 = Multiply(
   A=Multiply(A=Normalize(VectorInput=@n45), B=Constant3Vector() {constant=<rgba 1 1 0 0>})
   B=Add(A=ScalarParameter() {defaultValue=10, group=Noise, parameterName=WPOStrength}, B=Multiply(A=@n46, B=@n47) {constB=0.05})
  )
  StaticSwitchParameter(True=Add(A=Multiply(A=@n44, B=@n48) {constB=0}, B=@n11), False=@n11) {group=Hologram, parameterName=Is Hologram, sortPriority=1}
 ]
}
