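# Masked unlit billboard material for 2D icons on Creative devices, with parameters for item icon, device icon and colors.
# ItemIcon RGB shades between darkened and full background color, then a scaled, offset, clamp-sampled DeviceIcon tinted by IconColor is blended over it.
# Summed icon alphas form the opacity mask, and AlignMeshToTheCamera drives world position offset.
# The "Is Hologram" static switch bypasses MF_CreativeMoveTool when true, otherwise applies it.
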Material /Game/Creative/Devices/Generic/Materials/M_CP_Device_2DIcon_UserSet
graph {
 properties {bUseMaterialAttributes=true, blendMode=BLEND_Masked, pixelDepthOffsetMode=PDOM_Legacy, shadingModel=MSM_Unlit, usedShadingModels=Unlit}
 nodes [
  n1 = VectorParameter() {group=Background, parameterName=ColorBackground, defaultValue=<rgba 0 0.6223 1 1>}
  n2 = TextureSampleParameter2D() {
   group=Color
   parameterName=ItemIcon
   samplerSource=SSM_Clamp_WorldGroupSettings
   sortPriority=2
   texture=/Game/Creative/Devices/StatManager/Materials/Textures/T_CP_Device_StatManager_01_D
  }
  n3 = LinearInterpolate(
   A=Multiply(A=ScalarParameter() {defaultValue=0.25, group=Background, parameterName=ColorBackgroundLowMultiplier}, B=@n1.RGB)
   B=@n1.RGB
   Alpha=@n2.RGB
  )
  n4 = VectorParameter() {parameterName=IconColor, defaultValue=<rgba 1 1 1 1>}
  n5 = ScalarParameter() {defaultValue=-1, group=Device, parameterName=DeviceIconOffsetX}
  n6 = ScalarParameter() {defaultValue=-0.6, group=Device, parameterName=DeviceIconOffsetY}
  n7 = TextureSampleParameter2D!(
   UVs=Add(
    A=TextureCoordinate() {uTiling=3, vTiling=3}
    B=MaterialFunctionCall(X=@n5, Y=@n6) {materialFunction=&MakeFloat2}.Result
   )
   Level
   Apply View MipBias
  ) {
   group=Device
   mipValueMode=TMVM_MipLevel
   parameterName=DeviceIcon
   samplerSource=SSM_Clamp_WorldGroupSettings
   sortPriority=2
   texture=/Game/Athena/UI/InGame/Creative/PropertyEditor/Icon_Picker/Textures/T_UI_IconPicker_EMPTY
  }
  n8 = MakeMaterialAttributes!(
   BaseColor
   Metallic
   Specular
   Roughness
   Anisotropy
   EmissiveColor=LinearInterpolate(A=Multiply(A=@n3, B=@n2.A), B=Multiply(A=@n7.RGB, B=@n4.RGB), Alpha=@n7.A)
   Opacity
   OpacityMask=Saturate(None=Add(A=@n2.A, B=@n7.A))
   Normal
   Tangent
   WorldPositionOffset=MaterialFunctionCall!(
    WorldPosition
    PivotLocation
    WS Normals
    Custom Object Basis 1=Constant3Vector() {constant=<rgba 1 0 0 0>}
    Custom Object Basis 2=Constant3Vector() {constant=<rgba 0 0 -1 0>}
    Custom Object Basis 3=Constant3Vector() {constant=<rgba 0 1 0 0>}
   ) {materialFunction=/Engine/Functions/Engine_MaterialFunctions02/WorldPositionOffset/AlignMeshToTheCamera}.World Position Offset
   SubsurfaceColor
   ClearCoat
   ClearCoatRoughness
   AmbientOcclusion
   Refraction
   CustomizedUVs_0
   CustomizedUVs_1
   CustomizedUVs_2
   CustomizedUVs_3
   CustomizedUVs_4
   CustomizedUVs_5
   CustomizedUVs_6
   CustomizedUVs_7
   PixelDepthOffset
   ShadingModel
   Displacement
  )
  StaticSwitchParameter(True=@n8, False=MaterialFunctionCall(MA In=@n8) {materialFunction=&MF_CreativeMoveTool}.MA Out) {group=Hologram, parameterName=Is Hologram, sortPriority=1}
 ]
}
