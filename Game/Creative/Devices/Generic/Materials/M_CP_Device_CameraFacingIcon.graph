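# Unlit translucent master material for Creative device icons that turn to face the camera, drawn as a logo with glow or as a hologram.
# World position offset uses two-axis camera alignment or atan2-driven single-axis rotation, plus optional bobbing.
# The Is Hologram switch layers actor-anchored screen-space scanlines, flickering world-aligned noise and a TurnOn pulse into opacity and curve-atlas-colored emissive.
# An Is Opaque switch supports an opaque blend-mode variant.
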Material /Game/Creative/Devices/Generic/Materials/M_CP_Device_CameraFacingIcon
graph {
 properties {blendMode=BLEND_Translucent, pixelDepthOffsetMode=PDOM_Legacy, shadingModel=MSM_Unlit, usedShadingModels=Unlit}
 nodes [
  n1 = Divide(A=ObjectRadius(), B=ScalarParameter() {defaultValue=0.1, group=Noise, parameterName=NoiseSize})
  n2 = Multiply(A=Time(), B=VectorParameter() {group=Noise, parameterName=NoisePanSpeed, defaultValue=<rgba -0.08 0.07 -0.2 0>}.RGB)
  n3 = MaterialFunctionCall(
   TextureObject=TextureObject() {samplerType=SAMPLERTYPE_LinearColor, texture=&T_NoisePack_LinearColor}
   TextureSize=Multiply(A=VectorParameter() {group=Noise, parameterName=NoiseTileMult, defaultValue=<rgba 1 1 0.1 1>}.RGB, B=@n1)
   WorldPosition=Add(A=WorldPosition().XYZ, B=Multiply(A=@n1, B=@n2))
  ) {materialFunction=&WorldAlignedTexture}
  n4 = ActorPositionWS()
  n5 = MaterialFunctionCall(Float2=Round(None=Divide(A=@n4) {constB=2000})) {materialFunction=&BreakOutFloat2Components}
  n6 = Add(A=Time(), B=Add(A=@n5.R, B=@n5.R))
  n7 = Multiply(A=ScalarParameter() {defaultValue=0.5, group=Hologram, parameterName=FlickerSpeed}, B=@n6)
  n8 = AppendVector(A=Sine(None=@n7) {period=1.6}, B=Sine(None=@n7) {period=1.42})
  n9 = Sine(None=@n7) {period=1.51}
  n10 = Sine(None=@n7) {period=2.1}
  n11 = ScalarParameter() {defaultValue=0.5, group=Noise, parameterName=NoiseFrequency}
  n12 = Multiply(B=@n11) {constA=0.25}
  n13 = Add(A=Sine(None=@n7) {period=0.72}, B=Multiply(B=@n11) {constA=0.75}) {constB=0.75}
  n14 = Add(A=Multiply(B=@n11) {constA=0.9, constB=0.9}, B=Sine(None=@n7) {period=1.25}) {constB=0.9}
  n15 = Multiply(A=Multiply(A=Add(A=@n10, B=@n12), B=@n13), B=Multiply(A=@n13, B=Ceil(None=@n14)))
  n16 = ConstantBiasScale(None=Sine(None=@n6) {period=0.075})
  n17 = Multiply(A=Ceil(None=AppendVector(A=@n8, B=@n9)), B=Multiply(A=Saturate(None=@n15), B=LinearInterpolate(Alpha=@n16) {constA=0.7}))
  n18 = ScalarParameter() {defaultValue=2, group=Noise, parameterName=NoiseMult}
  n19 = Multiply(
   A=Multiply(A=Power(Base=@n3.XY Texture) {constExponent=4}, B=ScalarParameter() {defaultValue=2, group=Noise, parameterName=NoiseMaskMult}) {constB=20}
   B=Add(A=ScalarParameter() {group=Noise, parameterName=NoiseAdd}, B=Multiply(A=@n17, B=@n18))
  )
  n20 = WorldPosition()
  n21 = ComponentMask(None=MaterialFunctionCall() {materialFunction=&ObjectPivotPoint}.Object Pivot Location) {b=true}
  n22 = Add(A=@n21, B=ScalarParameter() {defaultValue=300, group=Hologram, parameterName=PulseHeight}) {constB=1600}
  n23 = ScalarParameter() {defaultValue=1, group=Hologram, parameterName=TurnOn}
  n24 = ScalarParameter() {defaultValue=33, group=Hologram, parameterName=PulseRadius}
  n25 = Multiply(A=@n24) {constB=0.1}
  n26 = Multiply(
   A=SphereMask(
    A=ComponentMask(None=@n20.XYZ) {b=true}
    B=LinearInterpolate(A=@n21, B=@n22, Alpha=@n23) {constA=-0.25, constB=1.2}
    Radius=LinearInterpolate(A=@n24, B=@n25, Alpha=@n23) {constA=500, constB=100}
   ) {hardnessPercent=0}
   B=OneMinus(None=@n23)
  )
  n27 = Add(A=Desaturation(None=Saturate(None=@n19)), B=Multiply(A=@n26) {constB=2})
  n28 = QualitySwitch(
   Default=MaterialFunctionCall(In=@n27, Contrast=ScalarParameter() {defaultValue=0.2, group=Noise, parameterName=NoiseContrast}) {materialFunction=&CheapContrast}.Result
   Low=Constant()
  )
  n29 = Saturate(None=@n28)
  n30 = Multiply(A=Constant() {r=60}, B=LinearInterpolate(Alpha=Ceil(None=@n29)) {constA=1, constB=1.25})
  n31 = ConstantBiasScale(None=ScreenPosition().ViewportUV) {bias=-0.5, scale=2}
  n32 = MaterialFunctionCall(Float2=ViewSize()) {materialFunction=&BreakOutFloat2Components}
  n33 = AppendVector(A=Constant() {r=1}, B=Divide(A=@n32.G, B=@n32.G))
  n34 = Multiply(
   A=Constant2Vector() {g=-1, r=1}
   B=MaterialFunctionCall(World Position=ActorPositionWS() {desc=Also use Object Position}) {materialFunction=&TransformToClipSpace}.Clip Space XY
  )
  n35 = Subtract(
   A=ConstantBiasScale(None=Multiply(A=@n31, B=@n33) {desc=Keep 1:1 aspect ratio regardless of screen aspect ratio}) {scale=2}
   B=Multiply(A=Multiply(A=@n33, B=@n34)) {constB=0.5, desc=Halve offsets to account for 0...1 UV space}
  )
  n36 = Add(
   A=MaterialFunctionCall(UVs=@n35, "Rotation Angle (0-1)"=Constant() {r=0.25}) {materialFunction=&CustomRotator}.Rotated Values
   B=Multiply(A=Time(), B=Constant() {r=-0.02})
  )
  n37 = ConstantBiasScale(None=Sine(None=Multiply(A=@n30, B=@n36) {constA=1}))
  n38 = Add(A=Constant() {r=0.5}, B=@n37)
  n39 = ScalarParameter() {defaultValue=-0.4, group=Noise, parameterName=NoiseBrightness}
  n40 = ComponentMask(None=Subtract(A=WorldPosition().XYZ, B=ObjectPositionWS())) {b=true}
  n41 = ObjectRadius()
  n42 = Saturate(None=ConstantBiasScale(None=Divide(A=@n40, B=@n41)))
  n43 = LinearInterpolate(
   A=Multiply(A=Multiply(A=@n39, B=@n29) {constB=0.1}, B=Saturate(None=@n38))
   B=LinearInterpolate(A=@n42, B=OneMinus(None=@n42), Alpha=ScalarParameter() {group=Hologram, parameterName=FlipGradient})
   Alpha=ScalarParameter() {defaultValue=0.4, group=Hologram, parameterName=VerticalGradientInflience}
  )
  Add(A=@n43, B=ScalarParameter() {group=Hologram, parameterName=Fill})
  Multiply(
   A=MaterialFunctionCall(Value=Time(), Period=ScalarParameter() {defaultValue=2, parameterName=Pulse Speed}) {materialFunction=&LinearSine}.Linear Sine
   B=ScalarParameter() {defaultValue=0.6, group=Color, parameterName=Max Glow Brightness}
  )
  ScalarParameter() {defaultValue=60, group=Hologram, parameterName=ScanlineTile}
  ScalarParameter() {defaultValue=0.6, group=Hologram, parameterName=ScanlineMinOpacity}
  ScalarParameter() {defaultValue=0.5, group=Hologram, parameterName=ScanlineColorMinInfluence}
  ScalarParameter() {defaultValue=-0.02, group=Hologram, parameterName=ScanlinePanSpeed}
  n44 = Constant3Vector() {constant=<rgba 1 0 0 0>}
  n45 = Constant3Vector() {constant=<rgba 0 0 -1 0>}
  n46 = Constant3Vector() {constant=<rgba 0 1 0 0>}
  n47 = CameraPositionWS()
  n48 = Constant3Vector() {constant=<rgba 1 1 0 0>}
  MaterialExpressionNormalize_0 = Normalize(VectorInput=Subtract(A=ObjectPositionWS(), B=Multiply(A=@n47, B=@n48)))
  n49 = Custom!(In=@MaterialExpressionNormalize_0) {
   code="float2 output; \r\nfloat2 Input1; \r\noutput = atan2 (In.y,In.x); \r\nreturn (output);"
   desc="float2 output; \r\nfloat2 Input1; \r\noutput = atan2 (In.y,In.x); \r\nreturn (output);"
   outputType=CMOT_Float2
   inputs [{inputName=In, input {expression=&self:MaterialExpressionNormalize_0}}]
  }
  n50 = Divide(A=Divide(A=@n49) {constB=3.142})
  n51 = MaterialFunctionCall() {materialFunction=&ObjectPivotPoint}
  n52 = VectorParameter() {group=WPO, parameterName=Pivot Offset, defaultValue=<rgba 0 0 0 1>}
  n53 = StaticSwitchParameter(
   True=MaterialFunctionCall!(
    WorldPosition
    PivotLocation
    WS Normals
    Custom Object Basis 1=Transform(None=@n44) {transformSourceType=TRANSFORMSOURCE_Local}
    Custom Object Basis 2=Transform(None=@n45) {transformSourceType=TRANSFORMSOURCE_Local}
    Custom Object Basis 3=Transform(None=@n46) {transformSourceType=TRANSFORMSOURCE_Local}
   ) {materialFunction=/Engine/Functions/Engine_MaterialFunctions02/WorldPositionOffset/AlignMeshToTheCamera}.World Position Offset
   False=RotateAboutAxis(
    NormalizedRotationAxis=ObjectOrientation()
    RotationAngle=Add(A=@n50) {constB=0.25}
    PivotPoint=Add(A=@n51.Object Pivot Location, B=@n52.RGB)
    Position=WorldPosition().XYZ
   )
  ) {group=WPO, parameterName=Use 2-Axis, sortPriority=1}
  n54 = VectorParameter() {group=WPO, parameterName=BobHeightMax, defaultValue=<rgba 0 0 10 0>}
  n55 = VectorParameter() {group=WPO, parameterName=BobHeightMin, defaultValue=<rgba 0 0 -2 0>}
  n56 = Sine(None=Multiply(A=Time(), B=ScalarParameter() {defaultValue=0.4, group=WPO, parameterName=WPOTimeScale}))
  n57 = StaticSwitchParameter(True=Add(A=@n53, B=LinearInterpolate(A=@n54.RGB, B=@n55.RGB, Alpha=@n56)), False=@n53) {group=WPO, parameterName=Bob?}
  n58 = Subtract(A=WorldPosition().XYZ, B=ObjectPositionWS())
  n59 = ObjectRadius()
  n60 = ScalarParameter() {defaultValue=0.025, group=Noise, parameterName=WPOStrengthFromObjectRadius}
  n61 = Multiply(
   A=Multiply(A=Normalize(VectorInput=@n58), B=Constant3Vector() {constant=<rgba 1 1 0 0>})
   B=Add(A=ScalarParameter() {defaultValue=10, group=Noise, parameterName=WPOStrength}, B=Multiply(A=@n59, B=@n60) {constB=0.05})
  )
  StaticSwitchParameter(True=Add(A=Multiply(A=@n27, B=@n61) {constB=0}, B=@n57), False=@n57) {group=Hologram, parameterName=Is Hologram, sortPriority=1}
  n62 = ScalarParameter() {defaultValue=3, group=Glow Mask, parameterName=Glow Density}
  n63 = ScalarParameter() {defaultValue=1, group=Glow Mask, parameterName=Glow Scale}
  n64 = ScalarParameter() {group=Logo, parameterName=Logo Offset U}
  n65 = ScalarParameter() {group=Logo, parameterName=Logo Offset V}
  n66 = TextureSampleParameter2D(UVs=Add(
    A=MaterialFunctionCall(Texture Scale=@n63) {materialFunction=&ScaleUVsByCenter}.UVs
    B=AppendVector(A=@n64, B=@n65)
   )) {
   group=Color
   parameterName=Texture
   samplerSource=SSM_Clamp_WorldGroupSettings
   sortPriority=2
   texture=/Game/Creative/Devices/StoreDevice/Textures/T_CP_Device_StoreDevice_Logo_01_D
  }
  n67 = StaticSwitchParameter(
   True=@n66.G
   False=MaterialFunctionCall(
    In=MaterialFunctionCall(Density=@n62) {desc=Radial Mask, materialFunction=&RadialGradientExponential}.RadialGradientExponential
    Contrast=ScalarParameter() {group=Glow Mask, parameterName=Glow Edge Contrast}
   ) {materialFunction=&CheapContrast}.Result
  ) {defaultValue=true, desc=Uses the G channel from the Logo Texture, group=Glow Mask, parameterName=UseCustomGlowMask, sortPriority=1}
  n68 = ScalarParameter() {defaultValue=1, group=Logo, parameterName=Logo Scale}
  n69 = TextureSampleParameter2D(UVs=Add(
    A=MaterialFunctionCall(Texture Scale=@n68) {materialFunction=&ScaleUVsByCenter}.UVs
    B=AppendVector(A=@n64, B=@n65)
   )) {
   group=Color
   parameterName=Texture
   samplerSource=SSM_Clamp_WorldGroupSettings
   sortPriority=2
   texture=&T_CP_Device_StoreDevice_Logo_01_D
  }
  n70 = LinearInterpolate(A=Multiply(A=@n67, B=ScalarParameter() {defaultValue=1, group=Glow Mask, parameterName=Glow Opacity}), Alpha=@n69.R)
  n71 = Add(A=Multiply(A=ScalarParameter() {defaultValue=1, group=Hologram, parameterName=TurnOn}, B=Clamp(None=@n43)), B=@n26) {constA=1}
  n72 = DepthFade(Opacity=@n71, FadeDistance=ScalarParameter() {defaultValue=5, group=Hologram, parameterName=FadeDistance})
  n73 = Constant() {r=0.6}
  n74 = Multiply(A=@n28) {constB=1.5}
  n75 = LinearInterpolate(
   A=Clamp(None=Subtract(A=@n73, B=@n74)) {minDefault=-0.25}
   B=Add(A=@n74)
   Alpha=MaterialFunctionCall(Float2=Power(Base=@n37) {constExponent=1.5}) {materialFunction=&BreakOutFloat2Components}.R
  ) {constA=0.5}
  n76 = Multiply(A=Saturate(None=Multiply(A=@n72, B=@n75)), B=ScalarParameter() {defaultValue=5, group=Hologram, parameterName=OpacityMult})
  n77 = TextureSampleParameter2D() {
   group=Color
   parameterName=Texture
   samplerSource=SSM_Clamp_WorldGroupSettings
   sortPriority=2
   texture=&T_CP_Device_StoreDevice_Logo_01_D
  }
  StaticSwitchParameter(
   True=Constant() {r=1}
   False=StaticSwitchParameter(
    True=Multiply(A=@n76, B=@n77.A)
    False=StaticSwitchParameter(True=@n69.A, False=@n70) {group=Color, parameterName=Use RGB Instead of Colorize, sortPriority=0}
   ) {group=Hologram, parameterName=Is Hologram, sortPriority=1}
  ) {desc="Please set Blend Mode = Opaque", group=Hologram, parameterName=Is Opaque, sortPriority=3}
  n78 = TextureSampleParameter2D() {
   group=Color
   parameterName=Texture
   samplerSource=SSM_Clamp_WorldGroupSettings
   sortPriority=2
   texture=&T_CP_Device_StoreDevice_Logo_01_D
  }
  n79 = Multiply(
   A=StaticSwitchParameter(True=OneMinus(None=@n78.A), False=@n78.A) {group=Color, parameterName=Invert Outline Alpha, sortPriority=42}
   B=ScalarParameter() {defaultValue=2, group=Color, parameterName=Alpha Outline Intensity, sortPriority=41}
  )
  n80 = Multiply(A=@n77.RGB, B=ScalarParameter() {defaultValue=1, group=Color, parameterName=Diffuse Brightness})
  n81 = Desaturation(
   None=MaterialFunctionCall!(In=Clamp(None=@n80), Contrast=ScalarParameter() {group=Color, parameterName=Diffuse Cheap Contrast}) {materialFunction=/Engine/Functions/Engine_MaterialFunctions01/ImageAdjustment/CheapContrast_RGB}.Result
   Fraction=Constant() {r=1}
  )
  n82 = CurveAtlasRowParameter(CurveTime=MaterialFunctionCall(Float2=Saturate(None=@n71)) {materialFunction=&BreakOutFloat2Components}.R) {
   atlas=&AthenaCurveAtlas01
   curve=/Game/Creative/Devices/Generic/LinearColorCurves/Curve_CP_Device_Hologram_01
   defaultValue=282
   desc=Curve Atlas
   group=Color
   parameterName=CurveAtlas
  }
  n83 = Multiply(
   A=Multiply(A=Multiply(A=@n81, B=@n82), B=ScalarParameter() {defaultValue=20, group=Hologram, parameterName=EmissiveMult}) {constB=2}
   B=VectorParameter() {group=Color, parameterName=Tint for Hologram, sortPriority=35, defaultValue=<rgba 0.6 0.6253 1 1>}.RGB
  )
  n84 = LinearInterpolate(B=@n42, Alpha=ScalarParameter() {defaultValue=0.25, group=Hologram, parameterName=Opaque Gradient Intensity}) {constA=1}
  n85 = LinearInterpolate(B=@n75, Alpha=ScalarParameter() {defaultValue=0.5, group=Hologram, parameterName=Opaque Scanlines Intensity}) {constA=1}
  n86 = StaticSwitchParameter(True=Multiply(A=@n83, B=Multiply(A=@n84, B=@n85)), False=@n83) {desc="Please set Blend Mode = Opaque", group=Hologram, parameterName=Is Opaque, sortPriority=3}
  n87 = StaticSwitchParameter(
   True=@n69.RGB
   False=LinearInterpolate(
    A=VectorParameter() {group=Color, parameterName=Glow Color, sortPriority=4, defaultValue=<rgba 0.0192 0.0784 0.3177 1>}.RGB
    B=VectorParameter() {group=Color, parameterName=Logo Color, sortPriority=2, defaultValue=<rgba 0.9047 1 1 1>}.RGB
    Alpha=@n69.R
   ) {constAlpha=1}
  ) {group=Color, parameterName=Use RGB Instead of Colorize, sortPriority=0}
  n88 = ScalarParameter() {defaultValue=1, group=Color, parameterName=RGB Brightness}
  StaticSwitchParameter(
   True=StaticSwitchParameter(True=Add(A=@n79, B=@n86), False=@n86) {defaultValue=true, group=Color, parameterName=Use Alpha as Outline Glow, sortPriority=40}
   False=MaterialFunctionCall(In=Multiply(A=@n87, B=@n88), Contrast=ScalarParameter() {group=Color, parameterName=RGB Cheap Contrast}) {materialFunction=&CheapContrast_RGB}.Result
  ) {group=Hologram, parameterName=Is Hologram, sortPriority=1}
  n89 = MaterialFunctionCall(Vector 2=ObjectBounds()) {materialFunction=/Engine/Functions/Engine_MaterialFunctions02/Utility/VectorLength}
  n90 = Distance(A=ActorPositionWS(), B=CameraPositionWS())
  StaticSwitchParameter(
   True=MaterialFunctionCall(UVs=@n35, Texture Scale=Divide(A=@n89.V2 Length, B=@n90)) {materialFunction=&ScaleUVsByCenter}.UVs
   False=@n35
  ) {defaultValue=true, group=Hologram, parameterName=ScanlineScaleByDistance, sortPriority=1}
  TextureSampleParameter2D() {
   group=Diffuse Texture Overlay
   parameterName=Diffuse Texture Overlay
   texture=/Game/Creative/Devices/Generic/Materials/Textures/T_CP_Device_GearIcon_01_D
  }
 ]
}
