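# Translucent, unlit, two-sided beam material for the base of Creative devices, with emissive color faded along the V gradient.
# Opacity combines that gradient with inverted Fresnel.
# The Is Hologram switch multiplies in depth-faded, screen-space scanlines anchored to the actor, and adds horizontal radial vertex displacement.
# World-aligned panning noise, gated by layered-sine flicker offset by actor position, drives both effects; low quality drops it from opacity.
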Material /Game/Creative/Devices/Generic/Materials/M_CP_Device_GenericBaseBeam_01
graph {
 properties {blendMode=BLEND_Translucent, pixelDepthOffsetMode=PDOM_Legacy, shadingModel=MSM_Unlit, twoSided=true, usedShadingModels=Unlit}
 nodes [
  n1 = SphereMask(
   A=MaterialFunctionCall() {materialFunction=&LinearGradient}.UGradient
   B=Constant() {r=0.5}
   Radius=Constant() {r=0.5}
   Hardness=Constant()
  )
  n2 = Constant() {r=2}
  n3 = OneMinus(None=MaterialFunctionCall() {materialFunction=&LinearGradient}.VGradient)
  n4 = Constant() {r=5}
  Multiply(
   A=LinearInterpolate(A=Power(Base=@n1, Exp=@n2), Alpha=Power(Base=@n3, Exp=@n4))
   B=Power(
    Base=MaterialFunctionCall() {materialFunction=&LinearGradient}.VGradient
    Exp=Constant() {r=2}
   )
  )
  Multiply(
   A=Multiply(
    A=VectorParameter() {parameterName=Beam Color, defaultValue=<rgba 0 0.8482 1 1>}.RGB
    B=ScalarParameter() {defaultValue=10, parameterName=Emissive Intensity}
   )
   B=MaterialFunctionCall() {materialFunction=&LinearGradient}.VGradient
  )
  n5 = TextureObject() {samplerType=SAMPLERTYPE_LinearColor, texture=&T_NoisePack_LinearColor}
  n6 = Divide(A=ObjectRadius(), B=ScalarParameter() {defaultValue=0.1, group=Noise, parameterName=NoiseSize})
  n7 = Multiply(A=VectorParameter() {group=Noise, parameterName=NoiseTileMult, defaultValue=<rgba 1 1 0.1 1>}.RGB, B=@n6)
  n8 = Time()
  n9 = VectorParameter() {group=Noise, parameterName=NoisePanSpeed, defaultValue=<rgba -0.08 0.07 -0.2 0>}
  n10 = Add(A=WorldPosition().XYZ, B=Multiply(A=@n6, B=Multiply(A=@n8, B=@n9.RGB)))
  n11 = Multiply(
   A=Power(Base=MaterialFunctionCall(TextureObject=@n5, TextureSize=@n7, WorldPosition=@n10) {materialFunction=&WorldAlignedTexture}.XY Texture) {constExponent=4}
   B=ScalarParameter() {defaultValue=2, group=Noise, parameterName=NoiseMaskMult}
  ) {constB=20}
  n12 = ActorPositionWS()
  n13 = MaterialFunctionCall(Float2=Round(None=Divide(A=@n12) {constB=2000})) {materialFunction=&BreakOutFloat2Components}
  n14 = Add(A=Time(), B=Add(A=@n13.R, B=@n13.R))
  n15 = Multiply(A=ScalarParameter() {defaultValue=0.5, group=Hologram, parameterName=FlickerSpeed}, B=@n14)
  n16 = Sine(None=@n15) {period=1.6}
  n17 = Sine(None=@n15) {period=1.42}
  n18 = Ceil(None=AppendVector(A=AppendVector(A=@n16, B=@n17), B=Sine(None=@n15) {period=1.51}))
  n19 = ScalarParameter() {defaultValue=0.5, group=Noise, parameterName=NoiseFrequency}
  n20 = Add(A=Sine(None=@n15) {period=0.72}, B=Multiply(B=@n19) {constA=0.75}) {constB=0.75}
  n21 = Multiply(A=Add(A=Sine(None=@n15) {period=2.1}, B=Multiply(B=@n19) {constA=0.25}), B=@n20)
  n22 = Multiply(B=@n19) {constA=0.9, constB=0.9}
  n23 = Sine(None=@n15) {period=1.25}
  n24 = Multiply(A=@n20, B=Ceil(None=Add(A=@n22, B=@n23) {constB=0.9}))
  n25 = Sine(None=@n14) {period=0.075}
  n26 = Multiply(A=Saturate(None=Multiply(A=@n21, B=@n24)), B=LinearInterpolate(Alpha=ConstantBiasScale(None=@n25)) {constA=0.7})
  n27 = Add(
   A=ScalarParameter() {group=Noise, parameterName=NoiseAdd}
   B=Multiply(A=Multiply(A=@n18, B=@n26), B=ScalarParameter() {defaultValue=2, group=Noise, parameterName=NoiseMult})
  )
  n28 = Desaturation(None=Saturate(None=Multiply(A=@n11, B=@n27)))
  n29 = WorldPosition()
  n30 = ObjectPositionWS()
  n31 = Multiply(A=Normalize(VectorInput=Subtract(A=@n29.XYZ, B=@n30)), B=Constant3Vector() {constant=<rgba 1 1 0 0>})
  n32 = Add(
   A=ScalarParameter() {defaultValue=10, group=Noise, parameterName=WPOStrength}
   B=Multiply(A=ObjectRadius(), B=ScalarParameter() {defaultValue=0.025, group=Noise, parameterName=WPOStrengthFromObjectRadius}) {constB=0.05}
  )
  StaticSwitchParameter(True=Multiply(A=@n28, B=Multiply(A=@n31, B=@n32)) {constB=0}, False=Constant()) {group=Hologram, parameterName=Is Hologram, sortPriority=1}
  n33 = Fresnel(ExponentIn=ScalarParameter() {defaultValue=0.5, parameterName=Fresnel Opacity Exp})
  n34 = Multiply(
   A=MaterialFunctionCall() {materialFunction=&LinearGradient}.VGradient
   B=Power(Base=OneMinus(None=@n33), Exp=ScalarParameter() {defaultValue=1, parameterName=Fresnel Power Exp})
  )
  n35 = QualitySwitch(
   Default=MaterialFunctionCall(In=@n28, Contrast=ScalarParameter() {defaultValue=0.2, group=Noise, parameterName=NoiseContrast}) {materialFunction=&CheapContrast}.Result
   Low=Constant()
  )
  n36 = Saturate(None=@n35)
  n37 = Multiply(A=Constant() {r=60}, B=LinearInterpolate(Alpha=Ceil(None=@n36)) {constA=1, constB=1.25})
  n38 = ScreenPosition()
  n39 = MaterialFunctionCall(Float2=ViewSize()) {materialFunction=&BreakOutFloat2Components}
  n40 = AppendVector(A=Constant() {r=1}, B=Divide(A=@n39.G, B=@n39.G))
  n41 = ConstantBiasScale(None=Multiply(A=ConstantBiasScale(None=@n38.ViewportUV) {bias=-0.5, scale=2}, B=@n40) {desc=Keep 1:1 aspect ratio regardless of screen aspect ratio}) {scale=2}
  n42 = Constant2Vector() {g=-1, r=1}
  n43 = MaterialFunctionCall(World Position=ActorPositionWS() {desc=Also use Object Position}) {materialFunction=&TransformToClipSpace}
  n44 = Multiply(A=Multiply(A=@n40, B=Multiply(A=@n42, B=@n43.Clip Space XY))) {constB=0.5, desc=Halve offsets to account for 0...1 UV space}
  n45 = Add(
   A=MaterialFunctionCall(UVs=Subtract(A=@n41, B=@n44), "Rotation Angle (0-1)"=Constant() {r=0.25}) {materialFunction=&CustomRotator}.Rotated Values
   B=Multiply(A=Time(), B=Constant() {r=-0.02})
  )
  n46 = ConstantBiasScale(None=Sine(None=Multiply(A=@n37, B=@n45) {constA=1}))
  n47 = Saturate(None=Add(A=Constant() {r=0.5}, B=@n46))
  n48 = Multiply(A=ScalarParameter() {defaultValue=-0.4, group=Noise, parameterName=NoiseBrightness}, B=@n36) {constB=0.1}
  n49 = ComponentMask(None=Subtract(A=WorldPosition().XYZ, B=ObjectPositionWS())) {b=true}
  n50 = ObjectRadius()
  n51 = Saturate(None=ConstantBiasScale(None=Divide(A=@n49, B=@n50)))
  n52 = OneMinus(None=@n51)
  n53 = ScalarParameter() {group=Hologram, parameterName=FlipGradient}
  n54 = Clamp(None=LinearInterpolate(
    A=Multiply(A=@n48, B=@n47)
    B=LinearInterpolate(A=@n51, B=@n52, Alpha=@n53)
    Alpha=ScalarParameter() {defaultValue=0.4, group=Hologram, parameterName=VerticalGradientInflience}
   ))
  n55 = ScalarParameter() {defaultValue=5, group=Hologram, parameterName=FadeDistance}
  n56 = Multiply(A=@n35) {constB=1.5}
  n57 = Clamp(None=Subtract(A=ScalarParameter() {defaultValue=0.6, group=Hologram, parameterName=ScanlineMinOpacity}, B=@n56)) {minDefault=-0.25}
  n58 = Add(A=@n56)
  n59 = MaterialFunctionCall(Float2=Power(Base=@n46) {constExponent=1.5}) {materialFunction=&BreakOutFloat2Components}
  n60 = Saturate(None=Multiply(A=DepthFade(Opacity=@n54, FadeDistance=@n55), B=LinearInterpolate(A=@n57, B=@n58, Alpha=@n59.R) {constA=0.5}))
  n61 = ScalarParameter() {defaultValue=5, group=Hologram, parameterName=OpacityMult}
  StaticSwitchParameter(True=Multiply(A=Multiply(A=@n60, B=@n61), B=@n34), False=@n34) {group=Hologram, parameterName=Is Hologram, sortPriority=1}
 ]
}
