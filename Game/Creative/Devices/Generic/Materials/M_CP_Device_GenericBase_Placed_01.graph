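# Default Lit material, by name for the placed generic base of Creative devices, with parameters for textures and colors.
# The Masks texture recolors Diffuse, with green blending in Base Color and red then blending in Beam Color on top.
# The red mask also isolates a glow built from Beam Color times Emissive intensity.
# SpecularMasks and Normals textures are sampled separately, with no further processing.
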Material /Game/Creative/Devices/Generic/Materials/M_CP_Device_GenericBase_Placed_01
graph {
 properties {pixelDepthOffsetMode=PDOM_Legacy, usedShadingModels=Default Lit}
 nodes [
  n1 = TextureObjectParameter() {parameterName=Diffuse, texture=&Black}
  n2 = TextureSampleParameter2D() {parameterName=Masks, texture=/Engine/EngineResources/Black_Low.Black}
  n3 = VectorParameter() {group=Color, parameterName=Beam Color, defaultValue=<rgba 0 0.8482 1 1>}
  LinearInterpolate(
   A=LinearInterpolate(
    A=TextureSample(Tex=@n1).RGB
    B=VectorParameter() {group=Color, parameterName=Base Color, defaultValue=<rgba 0.0745 0.1661 0.4349 1>}.RGB
    Alpha=@n2.G
   )
   B=@n3.RGB
   Alpha=@n2.R
  )
  Multiply(A=Multiply(A=@n3.RGB, B=ScalarParameter() {defaultValue=6, group=Color, parameterName=Emissive intensity}), B=@n2.R)
  n4 = TextureObjectParameter() {parameterName=SpecularMasks, samplerType=SAMPLERTYPE_Masks, sortPriority=1, texture=/Game/Global/Textures/Default/Blanks/Temp_specular_Mask}
  TextureSample(Tex=@n4) {samplerType=SAMPLERTYPE_Masks}
  TextureSample(Tex=TextureObjectParameter() {parameterName=Normals, samplerType=SAMPLERTYPE_Normal, sortPriority=2, texture=/Engine/EngineMaterials/FlatNormal}) {samplerType=SAMPLERTYPE_Normal}
 ]
}
