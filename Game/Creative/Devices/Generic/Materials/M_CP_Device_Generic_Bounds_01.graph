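# Translucent unlit material for Creative device bounds, drawing glowing sphere, disc or linear-gradient shapes chosen by static switches.
# Shape masks drive depth-faded opacity and colored edge glow.
# The IsHologram switch adds aspect-corrected screen-space scanlines anchored to the actor, position-seeded flickering world-aligned noise, a TurnOn pulse, curve-atlas color and noise-driven vertex offset.
# Hologram opacity also uses a custom-depth comparison to hide pixels behind custom-depth objects, bypassed on mobile.
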Material /Game/Creative/Devices/Generic/Materials/M_CP_Device_Generic_Bounds_01
graph {
 properties {
  allowTranslucentCustomDepthWrites=true
  blendMode=BLEND_Translucent
  opacityMaskClipValue=0
  pixelDepthOffsetMode=PDOM_Legacy
  shadingModel=MSM_Unlit
  usedShadingModels=Unlit
 }
 nodes [
  Constant() {r=60}
  Constant() {r=-0.02}
  n1 = ScalarParameter() {defaultValue=2, group=Shape, parameterName=Spherical Fresnel Exponent}
  n2 = Clamp(None=MaterialFunctionCall(Density=Constant() {r=50}) {materialFunction=&RadialGradientExponential}.RadialGradientExponential)
  n3 = Clamp(None=MaterialFunctionCall(
    Radius=ScalarParameter() {defaultValue=0.75, group=Shape, parameterName=Disc Radius}
    Density=ScalarParameter() {defaultValue=3, group=Shape, parameterName=Disc Density}
    Invert Density=StaticBool()
   ) {materialFunction=&RadialGradientExponential}.RadialGradientExponential)
  n4 = MaterialFunctionCall() {materialFunction=&LinearGradient}
  n5 = StaticSwitchParameter(True=@n4.UGradient, False=@n4.UGradient) {group=Shape, parameterName=Toggle Horizontal Gradient, sortPriority=10}
  n6 = ScalarParameter() {defaultValue=4, group=Shape, parameterName=Linear Gradient Exponent}
  n7 = Power(Base=@n5, Exp=@n6)
  n8 = Power(Base=OneMinus(None=@n5), Exp=@n6)
  n9 = StaticSwitchParameter(
   True=Power(Base=Fresnel(ExponentIn=@n1), Exp=ScalarParameter() {defaultValue=1.5, group=Shape, parameterName=Spherical Fresnel Power Exponent})
   False=StaticSwitchParameter(True=Subtract(A=@n2, B=@n3), False=Add(A=@n7, B=@n8)) {group=Shape, parameterName="Is Radial (Disc)", sortPriority=1}
  ) {group=Shape, parameterName="Is 360 (Sphere)", sortPriority=2}
  n10 = ParticleColor()
  n11 = ScalarParameter() {parameterName=UseParticleColor}
  n12 = ComponentMask(None=Subtract(A=WorldPosition().XYZ, B=ObjectPositionWS())) {b=true}
  n13 = ObjectRadius()
  n14 = Saturate(None=ConstantBiasScale(None=Divide(A=@n12, B=@n13)))
  n15 = OneMinus(None=@n14)
  n16 = ScalarParameter() {parameterName=FlipGradient}
  n17 = Divide(A=ObjectRadius(), B=ScalarParameter() {defaultValue=0.1, group=Noise, parameterName=NoiseSize})
  n18 = Multiply(A=Time(), B=VectorParameter() {group=Noise, parameterName=NoisePanSpeed, defaultValue=<rgba -0.08 0.07 -0.2 0>}.RGB)
  n19 = MaterialFunctionCall(
   TextureObject=TextureObject() {samplerType=SAMPLERTYPE_LinearColor, texture=&T_NoisePack_LinearColor}
   TextureSize=Multiply(A=VectorParameter() {group=Noise, parameterName=NoiseTileMult, defaultValue=<rgba 1 1 0.1 1>}.RGB, B=@n17)
   WorldPosition=Add(A=WorldPosition().XYZ, B=Multiply(A=@n17, B=@n18))
  ) {materialFunction=&WorldAlignedTexture}
  n20 = ActorPositionWS()
  n21 = MaterialFunctionCall(Float2=Round(None=Divide(A=@n20) {constB=2000})) {materialFunction=&BreakOutFloat2Components}
  n22 = Add(A=Time(), B=Add(A=@n21.R, B=@n21.R))
  n23 = Multiply(A=ScalarParameter() {defaultValue=0.5, parameterName=FlickerSpeed}, B=@n22)
  n24 = AppendVector(A=Sine(None=@n23) {period=1.6}, B=Sine(None=@n23) {period=1.42})
  n25 = Sine(None=@n23) {period=1.51}
  n26 = Sine(None=@n23) {period=2.1}
  n27 = ScalarParameter() {defaultValue=0.5, group=Noise, parameterName=NoiseFrequency}
  n28 = Multiply(B=@n27) {constA=0.25}
  n29 = Add(A=Sine(None=@n23) {period=0.72}, B=Multiply(B=@n27) {constA=0.75}) {constB=0.75}
  n30 = Add(A=Multiply(B=@n27) {constA=0.9, constB=0.9}, B=Sine(None=@n23) {period=1.25}) {constB=0.9}
  n31 = Multiply(A=Multiply(A=Add(A=@n26, B=@n28), B=@n29), B=Multiply(A=@n29, B=Ceil(None=@n30)))
  n32 = ConstantBiasScale(None=Sine(None=@n22) {period=0.075})
  n33 = Multiply(A=Ceil(None=AppendVector(A=@n24, B=@n25)), B=Multiply(A=Saturate(None=@n31), B=LinearInterpolate(Alpha=@n32) {constA=0.7}))
  n34 = ScalarParameter() {defaultValue=2, group=Noise, parameterName=NoiseMult}
  n35 = Multiply(
   A=Multiply(A=Power(Base=@n19.XY Texture) {constExponent=4}, B=ScalarParameter() {defaultValue=2, group=Noise, parameterName=NoiseMaskMult}) {constB=20}
   B=Add(A=ScalarParameter() {group=Noise, parameterName=NoiseAdd}, B=Multiply(A=@n33, B=@n34))
  )
  n36 = WorldPosition()
  n37 = ComponentMask(None=MaterialFunctionCall() {materialFunction=&ObjectPivotPoint}.Object Pivot Location) {b=true}
  n38 = Add(A=@n37, B=ScalarParameter() {defaultValue=300, parameterName=PulseHeight}) {constB=1600}
  n39 = ScalarParameter() {defaultValue=1, parameterName=TurnOn}
  n40 = ScalarParameter() {defaultValue=33, parameterName=PulseRadius}
  n41 = Multiply(A=@n40) {constB=0.1}
  n42 = Multiply(
   A=SphereMask(
    A=ComponentMask(None=@n36.XYZ) {b=true}
    B=LinearInterpolate(A=@n37, B=@n38, Alpha=@n39) {constA=-0.25, constB=1.2}
    Radius=LinearInterpolate(A=@n40, B=@n41, Alpha=@n39) {constA=500, constB=100}
   ) {hardnessPercent=0}
   B=OneMinus(None=@n39)
  )
  n43 = Add(A=Desaturation(None=Saturate(None=@n35)), B=Multiply(A=@n42) {constB=2})
  n44 = QualitySwitch(
   Default=MaterialFunctionCall(In=@n43, Contrast=ScalarParameter() {defaultValue=0.2, group=Noise, parameterName=NoiseContrast}) {materialFunction=&CheapContrast}.Result
   Low=Constant()
  )
  n45 = Saturate(None=@n44)
  n46 = Multiply(
   A=ScalarParameter() {defaultValue=60, group=Hologram Scanlines, parameterName=ScanlineTile}
   B=LinearInterpolate(Alpha=Ceil(None=@n45)) {constA=1, constB=1.25}
  )
  n47 = ScreenPosition()
  n48 = MaterialFunctionCall(Float2=ViewSize()) {materialFunction=&BreakOutFloat2Components}
  n49 = AppendVector(A=Constant() {r=1}, B=Divide(A=@n48.G, B=@n48.G))
  n50 = ConstantBiasScale(None=Multiply(A=ConstantBiasScale(None=@n47.ViewportUV) {bias=-0.5, scale=2}, B=@n49) {desc=Keep 1:1 aspect ratio regardless of screen aspect ratio}) {scale=2}
  n51 = Constant2Vector() {g=-1, r=1}
  n52 = MaterialFunctionCall(World Position=ActorPositionWS() {desc=Also use Object Position}) {materialFunction=&TransformToClipSpace}
  n53 = Multiply(A=Multiply(A=@n49, B=Multiply(A=@n51, B=@n52.Clip Space XY))) {constB=0.5, desc=Halve offsets to account for 0...1 UV space}
  n54 = Add(
   A=MaterialFunctionCall(UVs=Subtract(A=@n50, B=@n53), "Rotation Angle (0-1)"=Constant() {r=0.25}) {materialFunction=&CustomRotator}.Rotated Values
   B=Multiply(A=Time(), B=ScalarParameter() {defaultValue=-0.02, group=Hologram Scanlines, parameterName=ScanlinePanSpeed})
  )
  n55 = ConstantBiasScale(None=Sine(None=Multiply(A=@n46, B=@n54) {constA=1}))
  n56 = Add(A=Constant() {r=0.5}, B=@n55)
  n57 = Add(
   A=LinearInterpolate(
    A=Saturate(None=@n56)
    B=LinearInterpolate(A=@n14, B=@n15, Alpha=@n16)
    Alpha=ScalarParameter() {defaultValue=0.4, parameterName=VerticalGradientInflience}
   )
   B=ScalarParameter() {parameterName=Fill}
  )
  n58 = Add(A=Multiply(A=ScalarParameter() {defaultValue=1, parameterName=TurnOn}, B=Clamp(None=@n57)), B=@n42) {constA=1}
  n59 = Constant() {r=0.6}
  n60 = Multiply(A=@n44) {constB=1.5}
  n61 = LinearInterpolate(
   A=Clamp(None=Subtract(A=@n59, B=@n60)) {minDefault=-0.25}
   B=Add(A=@n60)
   Alpha=MaterialFunctionCall(Float2=Power(Base=@n55) {constExponent=1.5}) {materialFunction=&BreakOutFloat2Components}.R
  ) {constA=0.5}
  n62 = Multiply(A=Saturate(None=Multiply(A=@n58, B=@n61)), B=ScalarParameter() {defaultValue=1.25, parameterName=OpacityMult})
  n63 = Add(A=PixelDepth().R, B=ScalarParameter() {defaultValue=-0.25, parameterName=BiasSorting_Far})
  n64 = ComponentMask(None=SceneTexture() {sceneTextureId=PPI_CustomDepth}.Color) {r=true}
  n65 = Constant()
  n66 = Constant() {r=1}
  n67 = Multiply(
   A=Multiply(A=LinearInterpolate(B=@n10.A, Alpha=@n11) {constA=1, desc=Workaround for skel mesh bug}, B=Multiply(A=@n62))
   B=ShadingPathSwitch(Default=If(A=@n63, B=@n64, "A > B"=@n65, "A == B"=@n65, "A < B"=@n66), Mobile=Constant() {r=1}) {desc=Migrated from FeatureLevelSwitch}
  )
  DepthFade(
   Opacity=StaticSwitchParameter(True=Multiply(A=@n67, B=@n9), False=@n9) {parameterName=IsHologram, sortPriority=0}
   FadeDistance=ScalarParameter() {defaultValue=200, parameterName=FadeDistance}
  )
  n68 = DotProduct(A=VertexNormalWS(), B=Constant3Vector() {constant=<rgba 0 0 1 0>})
  LinearInterpolate(A=@n55, Alpha=Power(Base=Abs(None=@n68)) {constExponent=5}) {constB=0.25}
  n69 = Fresnel(ExponentIn=ScalarParameter() {defaultValue=1, group=Color, parameterName=Edge Fresnel Exponent, sortPriority=41}, Normal=VertexNormalWS())
  n70 = Clamp(None=MaterialFunctionCall(
    Radius=ScalarParameter() {defaultValue=0.7, group=Color, parameterName=Edge Radius, sortPriority=41}
    Density=ScalarParameter() {defaultValue=4, group=Color, parameterName=Edge Density, sortPriority=41}
    Invert Density=StaticBool()
   ) {materialFunction=&RadialGradientExponential}.RadialGradientExponential)
  n71 = MaterialFunctionCall() {materialFunction=&LinearGradient}
  n72 = StaticSwitchParameter(True=@n71.UGradient, False=@n71.UGradient) {group=Shape, parameterName=Toggle Horizontal Gradient, sortPriority=10}
  n73 = ScalarParameter() {defaultValue=4, group=Color, parameterName=Edge Linear Gradient Exponent, sortPriority=41}
  n74 = OneMinus(None=@n72)
  n75 = StaticSwitchParameter(True=OneMinus(None=@n70), False=Add(A=Power(Base=@n72, Exp=@n73), B=Power(Base=@n74, Exp=@n73))) {group=Shape, parameterName="Is Radial (Disc)", sortPriority=1}
  n76 = Multiply(
   A=MaterialFunctionCall(
    In=StaticSwitchParameter(True=@n69, False=@n75) {group=Shape, parameterName="Is 360 (Sphere)", sortPriority=2}
    Contrast=ScalarParameter() {defaultValue=0.5, group=Color, parameterName=Edge Fresnel Contrast, sortPriority=50}
   ) {materialFunction=&CheapContrast}.Result
   B=ScalarParameter() {defaultValue=5, group=Color, parameterName=Edge Intensity, sortPriority=40}
  )
  n77 = VectorParameter() {group=Color, parameterName=Edge Color, sortPriority=11, defaultValue=<rgba 0.3916 0.7991 1 1>}
  n78 = MaterialFunctionCall(Float2=Saturate(None=@n58)) {materialFunction=&BreakOutFloat2Components}
  n79 = CurveAtlasRowParameter(CurveTime=@n78.R) {
   atlas=&AthenaCurveAtlas01
   curve=/Game/Creative/Devices/Generic/LinearColorCurves/Curve_CP_Device_Hologram_01
   defaultValue=282
   desc=Curve Atlas
   group=Color
   parameterName=CurveAtlas
  }
  n80 = Multiply(
   A=Multiply(
    A=StaticSwitchParameter(True=@n79, False=@n78.R) {defaultValue=true, group=Color, parameterName=UseColorCurve, sortPriority=0}
    B=ScalarParameter() {defaultValue=2, parameterName=EmissiveMult}
   ) {constB=2}
   B=VectorParameter() {group=Color, parameterName=Tint for Hologram, sortPriority=10, defaultValue=<rgba 0.6 0.6253 1 1>}.RGB
  )
  n81 = VectorParameter() {group=Color, parameterName=Color, sortPriority=10, defaultValue=<rgba 0.1418 0.4318 0.612 1>}
  Reroute(None=Add(A=Multiply(A=@n76, B=@n77.RGB), B=StaticSwitchParameter(True=@n80, False=@n81.RGB) {parameterName=IsHologram, sortPriority=0}))
  ScalarParameter() {defaultValue=0.6, group=Hologram Scanlines, parameterName=MinScanlineOpacity}
  ScalarParameter() {defaultValue=0.7, group=Hologram Scanlines, parameterName=ScanlineColorMinInfluence}
  n82 = WorldPosition()
  n83 = ObjectPositionWS()
  n84 = Multiply(A=Normalize(VectorInput=Subtract(A=@n82.XYZ, B=@n83)), B=Constant3Vector() {constant=<rgba 1 1 0 0>})
  n85 = Add(
   A=ScalarParameter() {defaultValue=10, group=Noise, parameterName=WPOStrength}
   B=Multiply(A=ObjectRadius(), B=ScalarParameter() {defaultValue=0.025, group=Noise, parameterName=WPOStrengthFromObjectRadius}) {constB=0.05}
  )
  StaticSwitchParameter(True=Multiply(A=@n43, B=Multiply(A=@n84, B=@n85)) {constB=0}, False=Constant()) {parameterName=IsHologram, sortPriority=0}
  n86 = Constant() {r=50}
  Subtract(
   A=Clamp(None=MaterialFunctionCall(Density=@n86) {materialFunction=&RadialGradientExponential}.RadialGradientExponential)
   B=@n70
  )
 ]
}
