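# Material function for Creative devices that outputs a World Position Offset pushing a mesh along a chosen local forward axis, so it respects actor rotation.
# It transforms the Forward Vector input (local Y by default) out of local space, normalizes it and multiplies it by the scalar Offset Amount.
# Normalizing after the transform keeps actor scale from affecting the offset distance.
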MaterialFunction /Game/Creative/Devices/Generic/MaterialFunctions/MF_CP_Device_ForwardVectorOffset
graph {
 properties {description=Plugs into World Position Offset to Offset Mesh Forward Along Specified Axis such that it will work with Actor Rotation}
 nodes [
  n1 = FunctionInput(Preview=Constant3Vector() {constant=<rgba 0 1 0 0>}) {bUsePreviewValueAsDefault=true, inputName=Forward Vector, sortPriority=1, previewValue=<0 1 0 1>}
  n2 = Transform(None=@n1) {transformSourceType=TRANSFORMSOURCE_Local}
  n3 = FunctionInput(Preview=Constant()) {bUsePreviewValueAsDefault=true, inputName=Offset Amount, inputType=FunctionInput_Scalar, sortPriority=2, previewValue=<0 1 0 1>}
  FunctionOutput(None=Multiply(A=Normalize(VectorInput=@n2), B=@n3)) {outputName=WorldPositionOffset}
 ]
}
