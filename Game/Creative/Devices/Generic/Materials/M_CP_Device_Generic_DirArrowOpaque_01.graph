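# Opaque, Default Lit material for a directional arrow on generic Creative devices.
# A shared forward vector offset function takes a +Y forward vector and a Forward Offset parameter grouped under WPO, with a note that the asset should face negative Y.
# A Color parameter, teal by default, is multiplied by Emissive Intensity to form the glow color.
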Material /Game/Creative/Devices/Generic/Materials/M_CP_Device_Generic_DirArrowOpaque_01
graph {
 properties {allowTranslucentCustomDepthWrites=true, opacityMaskClipValue=0, pixelDepthOffsetMode=PDOM_Legacy, usedShadingModels=Default Lit}
 nodes [
  MaterialFunctionCall!(Forward Vector=Constant3Vector() {constant=<rgba 0 1 0 0>}, Offset Amount=ScalarParameter() {group=WPO, parameterName=Forward Offset}) {
   desc=Asset should be made facing Y negative
   materialFunction=/Game/Creative/Devices/Generic/MaterialFunctions/MF_CP_Device_ForwardVectorOffset
  }
  Multiply(
   A=VectorParameter() {parameterName=Color, defaultValue=<rgba 0.1356 0.4179 0.5271 1>}.RGB
   B=ScalarParameter() {defaultValue=0.25, parameterName=Emissive Intensity}
  ) {constB=0.25}
 ]
}
