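# Translucent unlit material for a generic Creative device direction arrow, tinted by a color parameter with an optional contrast-adjusted, desaturated texture overlay.
# Opacity adds contrast-adjusted Fresnel to a base value.
# A biased pixel depth versus custom depth test zeroes opacity behind custom depth (bypassed on mobile), with translucent custom depth writes allowed for sorting.
# A forward-vector function offsets vertices, expecting meshes facing negative Y.
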Material /Game/Creative/Devices/Generic/Materials/M_CP_Device_Generic_DirArrow_01
graph {
 properties {
  allowTranslucentCustomDepthWrites=true
  blendMode=BLEND_Translucent
  opacityMaskClipValue=0
  pixelDepthOffsetMode=PDOM_Legacy
  shadingModel=MSM_Unlit
  usedShadingModels=Unlit
 }
 nodes [
  MaterialFunctionCall!(Forward Vector=Constant3Vector() {constant=<rgba 0 1 0 0>}, Offset Amount=ScalarParameter() {group=WPO, parameterName=Forward Offset}) {
   desc=Asset should be made facing Y negative
   materialFunction=/Game/Creative/Devices/Generic/MaterialFunctions/MF_CP_Device_ForwardVectorOffset
  }
  n1 = MaterialFunctionCall(
   In=Fresnel(ExponentIn=ScalarParameter() {defaultValue=1, group=Opacity, parameterName=Fresnel Exponent})
   Contrast=ScalarParameter() {group=Opacity, parameterName=Fresnel Contrast}
  ) {materialFunction=&CheapContrast}
  n2 = ScalarParameter() {defaultValue=1, group=Opacity, parameterName=Fresnel Intensity}
  n3 = Add(A=PixelDepth().R, B=ScalarParameter() {defaultValue=-0.25, group=Sorting, parameterName=BiasSorting_Far})
  n4 = ComponentMask(None=SceneTexture() {sceneTextureId=PPI_CustomDepth}.Color) {r=true}
  n5 = Constant()
  n6 = Constant() {r=1}
  Multiply(
   A=Add(A=ScalarParameter() {parameterName=Base Opacity}, B=Multiply(A=@n1.Result, B=@n2))
   B=ShadingPathSwitch(Default=If(A=@n3, B=@n4, "A > B"=@n5, "A == B"=@n5, "A < B"=@n6), Mobile=Constant() {r=1}) {desc=Migrated from FeatureLevelSwitch}
  )
  n7 = TextureSampleParameter2D() {
   group=Diffuse Texture Overlay
   parameterName=Diffuse Texture Overlay
   texture=/Game/Creative/Devices/Generic/Materials/Textures/T_CP_Device_Generic_FlatArrow_01_D
  }
  n8 = LinearInterpolate(
   A=@n7.RGB
   B=MaterialFunctionCall(In=@n7.RGB, Contrast=ScalarParameter() {group=Diffuse Texture Overlay, parameterName=Diffuse Contribution Cheap Contrast}) {materialFunction=&CheapContrast}.Result
   Alpha=ScalarParameter() {group=Diffuse Texture Overlay, parameterName=Diffuse Overlay Use Cheap Contrast}
  )
  n9 = ScalarParameter() {defaultValue=1, group=Diffuse Texture Overlay, parameterName=Diffuse Contribution Desaturation}
  n10 = VectorParameter() {parameterName=Color, defaultValue=<rgba 0.7682 0.9643 1 1>}
  StaticSwitchParameter(
   True=LinearInterpolate(
    A=@n10.RGB
    B=Desaturation(None=@n8, Fraction=@n9)
    Alpha=ScalarParameter() {defaultValue=1, group=Diffuse Texture Overlay, parameterName=Lerp Diffuse Overlay}
   )
   False=@n10.RGB
  ) {group=Diffuse Texture Overlay, parameterName=UseDiffuseTextureOverlay, sortPriority=0}
 ]
}
