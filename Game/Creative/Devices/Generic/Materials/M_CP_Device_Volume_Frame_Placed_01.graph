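# Unlit, two-sided material that, by its name, draws the frame of a placed Creative device volume.
# TwoSidedSign blends inside and outside colors, alongside a world-aligned transition texture tinted by the outside color.
# A toggleable camera-distance fade and a scaled rock detail normal form separate branches.
# Rounded bounding-box UVWs select signed SizeBias offsets per side, transformed from local space, apparently to resize the frame.
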Material /Game/Creative/Devices/Generic/Materials/M_CP_Device_Volume_Frame_Placed_01
graph {
 properties {pixelDepthOffsetMode=PDOM_Legacy, shadingModel=MSM_Unlit, twoSided=true, usedShadingModels=Unlit}
 nodes [
  n1 = MaterialFunctionCall(
   TextureObject=TextureObject() {texture=/Game/Weapons/Prototype/ObjectMover/Textures/T_Movetool_Transition}
   TextureSize=VectorParameter() {parameterName="TextureSize (V3)", defaultValue=<rgba 32 32 32 0>}.RGB
   WorldPosition=WorldPosition().XYZ
  ) {materialFunction=&WorldAlignedTexture}
  n2 = VectorParameter() {desc=FF0004FF, parameterName=ColorOutside, defaultValue=<rgba 1 0 0.0157 1>}
  Add(A=Multiply(A=ComponentMask(None=@n1.XYZ Texture) {r=true}, B=@n2.RGB), B=@n2.RGB)
  LinearInterpolate(
   A=VectorParameter() {desc=FFB310FF, parameterName=ColorInside, defaultValue=<rgba 1 0 0.0157 1>}.RGB
   B=@n2.RGB
   Alpha=Saturate(None=TwoSidedSign())
  )
  n3 = CameraPositionWS()
  n4 = WorldPosition()
  n5 = Divide(
   A=Add(A=Distance(A=@n3, B=@n4.XYZ), B=ScalarParameter() {defaultValue=-100, group=Distance Fade, parameterName=DistanceBias})
   B=ScalarParameter() {defaultValue=1200, group=Distance Fade, parameterName=FadeScale}
  )
  LinearInterpolate(B=OneMinus(None=Saturate(None=@n5)), Alpha=ScalarParameter() {group=Distance Fade, parameterName=Enable Distance Fade, sortPriority=1}) {constA=1}
  n6 = VectorParameter() {parameterName=WorldScale, defaultValue=<rgba 20 20 20 0>}
  Multiply(
   A=TextureSample(UVs=MaterialFunctionCall!(WorldScale=@n6.RGB, TransformToLocalSpace) {materialFunction=/Game/Items/Traps/Materials/MF_Creative_WorldAlignedUVs}.Result) {samplerType=SAMPLERTYPE_Normal, texture=/Game/Environments/Sets/Arid/Textures/RockDetail_normal}.RGB
   B=VectorParameter() {parameterName=NormalStrength, defaultValue=<rgba 0.4 0.4 1 0>}.RGB
  )
  n7 = VectorParameter() {parameterName=SizeBias, defaultValue=<rgba 0 0 0 1>}
  n8 = MaterialFunctionCall() {materialFunction=&BoundingBoxBased_0-1_UVW}
  n9 = MaterialFunctionCall(X=Multiply(A=@n7.R) {constB=-1}, Y=Multiply(A=@n7.G) {constB=-1}, Z=Multiply(A=@n7.RGB, B=Round(None=@n8.B))) {materialFunction=&MakeFloat3}
  n10 = Round(None=@n8.RGB)
  Reroute(None=Transform(None=LinearInterpolate(A=@n9.Result, B=@n7.RGB, Alpha=@n10)) {transformSourceType=TRANSFORMSOURCE_Local})
 ]
}
